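# Physics asset for the dynamic collar of the SkippingClouds medium male player body.
# Box bodies on back, side and front collar bones hang from a tiny spine_05 anchor, alongside capsules on clavicles, neck and head that approximate the upper body.
# Collar constraints to the spine have very tight twist limits, with front pieces also narrow on one swing axis.
# Side pieces also link to the back piece with free angular constraints, as do the body capsules along their chain.
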PhysicsAsset /Game/Characters/Player/Male/Medium/Bodies/M_MED_SkippingClouds/Meshes/M_MED_SkippingClouds_Dyn
bodies {
 clavicle_l [{
   shapeType=Capsule
   length=2.5812
   radius=8.2659
   shapeName=spine_05_capsule
   center=<15.0731 -1.6607 -1.8683>
   rotation=<rot 73.8141 -0.0006 -0.0006>
  }]
 clavicle_r [{
   shapeType=Capsule
   length=2.5812
   radius=8.2659
   shapeName=spine_05_capsule
   center=<-15.0731 1.6607 1.8683>
   rotation=<rot 73.8141 -0.0006 179.9994>
  }]
 dyn_collar_bk [{
   shapeType=Box
   extentX=17.1243
   extentY=11.8011
   extentZ=5.1981
   shapeName=dyn_collar_bk_box
   center=<5.9759 0 1.8965>
   rotation=<rot -1.1932 0 0>
  }]
 dyn_collar_fr_bott_l [{shapeType=Box, extentX=7.5, extentY=7.5, extentZ=2.0043, shapeName=dyn_collar_bk_box, center=<2.4903 0 0>}]
 dyn_collar_fr_bott_r [{shapeType=Box, extentX=7.5, extentY=7.5, extentZ=2.0043, shapeName=dyn_collar_bk_box, center=<2.4903 0 0>}]
 dyn_collar_fr_l [{shapeType=Box, extentX=7.5, extentY=10.9579, extentZ=2.0043, shapeName=dyn_collar_bk_box, center=<2.4903 0 0>}]
 dyn_collar_fr_r [{shapeType=Box, extentX=7.5, extentY=10.9579, extentZ=2.0043, shapeName=dyn_collar_bk_box, center=<2.4903 0 0>}]
 dyn_collar_l [{shapeType=Box, extentX=15.9797, extentY=12.9571, extentZ=7.1644, shapeName=dyn_collar_bk_box, center=<7.2287 0 2.7829>}]
 dyn_collar_r [{shapeType=Box, extentX=15.9797, extentY=12.9571, extentZ=7.1644, shapeName=dyn_collar_bk_box, center=<7.2287 0 2.7829>}]
 head [
  {shapeType=Capsule, length=9.1592, radius=10.7045, shapeName=spine_05_capsule, center=<7.2167 0.1044 0>, rotation=<rot 90 0 0>}
  {shapeType=Capsule, length=7.5195, radius=8.2923, shapeName=head_capsule, center=<4.0451 4.1885 0>, rotation=<rot -90 180 179.9999>}
 ]
 neck_01 [{shapeType=Capsule, length=2.8703, radius=5.7816, shapeName=spine_05_capsule, center=<5.2093 0 0>, rotation=<rot 90 0 0>}]
 neck_02 [{shapeType=Capsule, length=0.5677, radius=8.0278, shapeName=spine_05_capsule, center=<6.134 0.0522 -0.1631>, rotation=<rot 90 -90 -91.3343>}]
 spine_05 [{shapeType=Capsule, length=0.1, radius=0.1, shapeName=spine_05_capsule}]
}
constraints [
 {bone1Name=dyn_collar_bk, bone2Name=spine_05, swing1LimitDegrees=30, swing2LimitDegrees=60, twistLimitDegrees=1}
 {bone1Name=dyn_collar_fr_bott_l, bone2Name=spine_05, swing1LimitDegrees=2, swing2LimitDegrees=60, twistLimitDegrees=1}
 {bone1Name=dyn_collar_fr_bott_r, bone2Name=spine_05, swing1LimitDegrees=2, swing2LimitDegrees=60, twistLimitDegrees=1}
 {bone1Name=dyn_collar_fr_l, bone2Name=spine_05, swing1LimitDegrees=2, swing2LimitDegrees=60, twistLimitDegrees=1}
 {bone1Name=dyn_collar_fr_r, bone2Name=spine_05, swing1LimitDegrees=2, swing2LimitDegrees=60, twistLimitDegrees=1}
 {bone1Name=dyn_collar_l, bone2Name=spine_05, swing1LimitDegrees=30, swing2LimitDegrees=60, twistLimitDegrees=1}
 {bone1Name=dyn_collar_r, bone2Name=spine_05, swing1LimitDegrees=30, swing2LimitDegrees=60, twistLimitDegrees=1}
 {
  bone1Name=neck_01
  bone2Name=spine_05
  swing1LimitDegrees=45
  swing1Motion=Free
  swing2LimitDegrees=45
  swing2Motion=Free
  twistLimitDegrees=45
  twistMotion=Free
 }
 {
  bone1Name=neck_02
  bone2Name=neck_01
  swing1LimitDegrees=45
  swing1Motion=Free
  swing2LimitDegrees=45
  swing2Motion=Free
  twistLimitDegrees=45
  twistMotion=Free
 }
 {
  bone1Name=head
  bone2Name=neck_02
  swing1LimitDegrees=45
  swing1Motion=Free
  swing2LimitDegrees=45
  swing2Motion=Free
  twistLimitDegrees=45
  twistMotion=Free
 }
 {
  bone1Name=dyn_collar_r
  bone2Name=dyn_collar_bk
  swing1LimitDegrees=45
  swing1Motion=Free
  swing2LimitDegrees=45
  swing2Motion=Free
  twistLimitDegrees=45
  twistMotion=Free
 }
 {
  bone1Name=dyn_collar_l
  bone2Name=dyn_collar_bk
  swing1LimitDegrees=45
  swing1Motion=Free
  swing2LimitDegrees=45
  swing2Motion=Free
  twistLimitDegrees=45
  twistMotion=Free
 }
 {
  bone1Name=clavicle_l
  bone2Name=spine_05
  swing1LimitDegrees=45
  swing1Motion=Free
  swing2LimitDegrees=45
  swing2Motion=Free
  twistLimitDegrees=45
  twistMotion=Free
 }
 {
  bone1Name=clavicle_r
  bone2Name=spine_05
  swing1LimitDegrees=45
  swing1Motion=Free
  swing2LimitDegrees=45
  swing2Motion=Free
  twistLimitDegrees=45
  twistMotion=Free
 }
]
